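; EQS context that provides a single location on the ground beneath the querier, searching down by the absolute value of a configurable distance.
; For a FortPawn with a capsule, it sweeps that capsule downward and returns the bottom of the capsule at the hit.
; Otherwise a fallback line trace returns its impact point.
; Either trace returns its end point on a miss, and an invalid querier yields the zero vector.
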(fn ProvideSingleLocation
  (params (QuerierObject :type Object%20Reference) (QuerierActor :type Actor%20Object%20Reference))
  (bind _tracedowndistancefromquerierlocation (Variables|Default|GetTraceDownDistanceFromQuerierLocation))
  (bind _returnvalue (Math|Float|Absolute(Float) _tracedowndistancefromquerierlocation))
  (bind _returnvalue_2 (Math|Vector|MakeVector 0.0 0.0 _returnvalue))
  (bind _querieractor (Variables|GetQuerierActor)
    (:then
      (bind _asfort_pawn (Utilities|Casting|CastToFortPawn _querieractor)
        (:then
          (bind _capsulecomponent (Variables|/Script/Engine.Character|GetCapsuleComponent _asfort_pawn)
            (:then
              (bind (_outhit _returnvalue_3) (Collision|CapsuleTraceByChannel (Transformation|GetWorldLocation _capsulecomponent) (- (Transformation|GetWorldLocation _capsulecomponent) _returnvalue_2) (Components|Capsule|GetScaledCapsuleRadius _capsulecomponent) (Components|Capsule|GetScaledCapsuleHalfHeight _capsulecomponent) "TraceTypeQuery13" false (Utilities|Array|MakeArray (Variables|GetQuerierActor)) "None" true "(R=1.000000,G=0.000000,B=0.000000,A=1.000000)" "(R=0.000000,G=1.000000,B=0.000000,A=1.000000)" 1e+01 :!pinned true)) ; Capsule Sweep for FortPawns (ideal situation)
              (if _returnvalue_3
                (return (- (.Location (Collision|BreakHitResult _outhit)) (Math|Vector|MakeVector 0.0 0.0 (Components|Capsule|GetScaledCapsuleHalfHeight _capsulecomponent))))
                (else
                  (return (.TraceEnd (Collision|BreakHitResult _outhit))))))
            (:else
              _)))
        (:CastFailed
          _)) ; If FortPawn, sweep a Capsule\nIf not, just do a trace.
      (bind (_outhit_2 _returnvalue_4) (Collision|LineTraceByChannel (Transformation|GetWorldLocation _capsulecomponent) (- (Transformation|GetWorldLocation _capsulecomponent) _returnvalue_2) "TraceTypeQuery13" false (Utilities|Array|MakeArray (Variables|GetQuerierActor)) "None" true "(R=1.000000,G=0.000000,B=0.000000,A=1.000000)" "(R=0.000000,G=1.000000,B=0.000000,A=1.000000)" 1e+01 :!pinned true)) ; Fallback Line Trace if NOT a FortPawn with a Capsule
      (if _returnvalue_4
        (return (.ImpactPoint (Collision|BreakHitResult _outhit_2)))
        (else
          (return (.TraceEnd (Collision|BreakHitResult _outhit_2))))))
    (:else
      (return "0, 0, 0"))))
