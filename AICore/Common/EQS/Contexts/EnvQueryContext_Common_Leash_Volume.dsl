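; EQS context that supplies the querier's AI leash volume as a single actor for environment queries.
; It gets the querier's AI controller, finds the FortAthenaLeashComponent on that controller, and reads its leash volume.
; Validity checks at each step return no actor when the controller, component, or volume is missing.
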(fn ProvideSingleActor
  (params (QuerierObject :type Object%20Reference) (QuerierActor :type Actor%20Object%20Reference))
  (bind _returnvalue (AI|GetAIController QuerierActor))
  (bind _returnvalue_2 (Actor|GetComponentbyClass _returnvalue "/Script/FortniteAI.FortAthenaLeashComponent"))
  (bind _returnvalue_3 (Class|FortAthenaLeashComponent|GetLeashVolume _returnvalue_2))
  (Utilities|IsValid _returnvalue
    (:Is%20Valid
      (Utilities|IsValid _returnvalue_2
        (:Is%20Valid
          (Utilities|IsValid _returnvalue_3
            (:Is%20Valid
              (return _returnvalue_3))
            (:Is%20Not%20Valid
              (return 0))))
        (:Is%20Not%20Valid
          (return 0))))
    (:Is%20Not%20Valid
      (return 0))))
